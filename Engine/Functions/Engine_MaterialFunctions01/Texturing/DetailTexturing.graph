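# DetailTexturing is a library material function that layers tiling detail diffuse and normal textures onto a material's Diffuse and Normal inputs.
# Both textures share texture coordinates multiplied by Scale.
# The detail diffuse, meant as a linear high-pass map, is offset by 0.5, blended from white by DiffuseIntensity, multiplied into Diffuse, then clamped.
# The detail normal's XY is scaled by NormalIntensity, its Z zeroed, then added to Normal without renormalization.
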MaterialFunction /Engine/Functions/Engine_MaterialFunctions01/Texturing/DetailTexturing
graph {
 properties {bExposeToLibrary=true, libraryCategoriesText [Misc,Texturing]}
 nodes [
  n1 = FunctionInput() {inputName=Diffuse, previewValue=<0.5 0.5 0.5 1>}
  n2 = FunctionInput() {bUsePreviewValueAsDefault=true, inputName=Scale, inputType=FunctionInput_Scalar, sortPriority=-1, previewValue=<2 0 0 1>}
  n3 = Multiply(A=TextureCoordinate(), B=@n2)
  n4 = FunctionInput(Preview=TextureObject() {texture=null}) {
   description=Should have highpass run on it and have sRGB unchecked
   inputName=DetailDiffuse
   inputType=FunctionInput_Texture2D
   sortPriority=1
  }
  n5 = FunctionInput() {bUsePreviewValueAsDefault=true, inputName=DiffuseIntensity, inputType=FunctionInput_Scalar, sortPriority=2, previewValue=<1 0 0 1>}
  n6 = LinearInterpolate(A=Constant() {r=1}, B=Add(A=Constant() {r=0.5}, B=TextureSample(UVs=@n3, Tex=@n4).RGB), Alpha=@n5)
  FunctionOutput(None=Clamp(None=Multiply(A=@n1, B=@n6))) {description=Multiply against your diffuse, outputName=Diffuse}
  n7 = TextureSample(
   UVs=@n3
   Tex=FunctionInput(Preview=TextureObject() {texture=null}) {inputName=DetailNormal, inputType=FunctionInput_Texture2D, sortPriority=4}
  ) {samplerType=SAMPLERTYPE_Normal}
  n8 = FunctionInput() {bUsePreviewValueAsDefault=true, inputName=NormalIntensity, inputType=FunctionInput_Scalar, sortPriority=5, previewValue=<1 0 0 1>}
  n9 = AppendVector(A=AppendVector(A=@n8, B=@n8), B=Constant())
  FunctionOutput(None=Add(A=Multiply(A=@n7.RGB, B=@n9), B=FunctionInput() {bUsePreviewValueAsDefault=true, inputName=Normal, sortPriority=3, previewValue=<0 0 1 1>})) {description="Add to  your normal", outputName=Normal, sortPriority=1}
 ]
}
