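# Translucent deferred decal material for Helios claw marks, with parameterized metallic, specular, roughness and opacity.
# Switchable detail texturing layers detail maps over the diffuse and normal textures, with an optional time-of-day tint on diffuse.
# The graph also carries burn-decal logic: a UV-distorted mask, slope masking and a panning-cloud emissive glow shaped by sphere masks and decal lifetime.
# Feature-level switches provide simplified ES2/ES3.1 alternatives for several channels.
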Material /Game/Environments/Helios/Props/ClawMarks/Materials/M_Decal_WithNormal
graph {
 properties {
  bAutomaticallySetUsageInEditor=false
  blendMode=BLEND_Translucent
  materialDomain=MD_DeferredDecal
  pixelDepthOffsetMode=PDOM_Legacy
  usedShadingModels=Default Lit
 }
 nodes [
  Constant2Vector() {g=1.6, r=1.1}
  Constant3Vector()
  n1 = VectorParameter() {parameterName=Simple_Diffuse_Color, defaultValue=<rgba 0.5 0.5 0.5 1>}
  n2 = ScalarParameter() {defaultValue=1, parameterName=Detail_Intensity}
  n3 = MaterialFunctionCall!(
   Scale=ScalarParameter() {defaultValue=1, parameterName=Detail_Tiling}
   Diffuse=@n1.RGB
   DetailDiffuse=TextureObjectParameter() {parameterName=Detail_Diffuse}
   DiffuseIntensity=@n2
   Normal=TextureSampleParameter2D() {parameterName=Normal, samplerType=SAMPLERTYPE_Normal, texture=/Game/Environments/Helios/Props/ClawMarks/Textures/T_ClawMarks_N}.RGB
   DetailNormal=TextureObjectParameter() {parameterName=Detail_Normal, samplerType=SAMPLERTYPE_Normal, texture=/Game/Athena/Environments/DudeBro/T_Dudebro_Noise_N_01}
   NormalIntensity=@n2
  ) {materialFunction=/Engine/Functions/Engine_MaterialFunctions01/Texturing/DetailTexturing}
  n4 = TextureSampleParameter2D() {parameterName=Mobile_Diffuse, texture=/Game/Environments/Helios/Props/ClawMarks/Textures/T_ClawMarks_D_Metal}
  n5 = Multiply(
   A=@n4.RGB
   B=MaterialFunctionCall(
    Color_Morning=VectorParameter() {parameterName=Morning, defaultValue=<rgba 0.1771 0.0891 0.0577 1>}.RGB
    Color_Day=VectorParameter() {parameterName=Day, defaultValue=<rgba 0.6458 0.6458 0.6458 1>}.RGB
    Color_Evening=VectorParameter() {parameterName=Evening, defaultValue=<rgba 0.2448 0.1364 0.1176 1>}.RGB
    Color_Night=VectorParameter() {parameterName=Night, defaultValue=<rgba 0.1864 0.3986 0.5833 1>}.RGB
   ) {materialFunction=&Color_Time_of_Day}.Result
  )
  FeatureLevelSwitch(Default=@n3.Diffuse, ES2=@n5, ES3_1=@n5)
  n6 = Constant()
  FeatureLevelSwitch(Default=ScalarParameter() {parameterName=Metallic}, ES2=@n6, ES3_1=@n6)
  n7 = Constant()
  FeatureLevelSwitch(Default=ScalarParameter() {defaultValue=0.1, parameterName=Spec}, ES2=@n7, ES3_1=@n7)
  n8 = Constant()
  FeatureLevelSwitch(Default=ScalarParameter() {defaultValue=2, parameterName=Roughness}, ES2=@n8, ES3_1=@n8)
  n9 = VectorParameter() {parameterName=ColorLow, defaultValue=<rgba 1 0.3294 0.1183 1>}
  n10 = VectorParameter() {parameterName=ColorHigh, defaultValue=<rgba 1 0.6484 0.217 1>}
  n11 = TextureSample(UVs=Panner(Coordinate=TextureCoordinate() {uTiling=1.2, vTiling=1.5}) {speedX=0.05, speedY=0.05}) {samplerType=SAMPLERTYPE_LinearColor, texture=/Game/VisualThreatManager/StormVisuals/Textures/Clouds_VectorDisplacement}
  n12 = ComponentMask(None=TextureSample() {texture=/Game/Global/Textures/Default/Noise/3xPerlin_M}.RGB) {g=true, r=true}
  n13 = LinearInterpolate(Alpha=ScalarParameter() {parameterName=UV_Distortion}) {constB=0.3}
  n14 = TextureSampleParameter2D(UVs=Add(A=TextureCoordinate(), B=Multiply(A=@n12, B=@n13))) {parameterName=BaseColorTexture, samplerType=SAMPLERTYPE_Masks, texture=/Game/Environments/Props/WildWest/Textures/T_BurnDecal}
  n15 = ScalarParameter() {parameterName=UseTextureAlpha}
  n16 = StaticSwitchParameter(True=@n4.A, False=OneMinus(None=LinearInterpolate(A=@n14.R, B=@n14.R, Alpha=@n15))) {parameterName=Use_Unique_Mask}
  n17 = MaterialFunctionCall(
   UVs=TextureCoordinate()
   Texture Scale=AppendVector(
    A=ScalarParameter() {defaultValue=1.1, parameterName=Emissive_U_Scale}
    B=ScalarParameter() {defaultValue=1.6, parameterName=Emissive_V_Scale}
   )
  ) {materialFunction=&ScaleUVsByCenter}
  n18 = DecalLifetimeOpacity()
  n19 = OneMinus(None=@n18.Opacity)
  n20 = AppendVector(A=LinearInterpolate(Alpha=@n19) {constA=0.45, constB=0.2}, B=LinearInterpolate(Alpha=@n19) {constA=0.4, constB=0.5})
  n21 = LinearInterpolate(Alpha=@n19) {constA=0.2}
  n22 = LinearInterpolate(Alpha=@n11.G) {constB=0.4}
  n23 = Power(Base=@n18.Opacity) {constExponent=5}
  n24 = Subtract(
   A=SphereMask(A=TextureCoordinate(), B=Constant() {r=0.5}, Radius=LinearInterpolate(Alpha=@n11.R) {constA=0.45, constB=0.55}) {attenuationRadius=0.5, hardnessPercent=0}
   B=LinearInterpolate(A=SphereMask(A=@n17.UVs, B=@n20, Radius=@n21, Hardness=@n22) {attenuationRadius=0.4, hardnessPercent=40}, Alpha=@n23) {constB=0}
  )
  n25 = Multiply(A=Multiply(A=LinearInterpolate(A=@n9.RGB, B=@n10.RGB, Alpha=@n11.G), B=@n11.R), B=Multiply(A=@n16, B=Saturate(None=@n24)))
  n26 = Multiply(A=@n25) {constB=5}
  Multiply(
   A=ScalarParameter() {parameterName=OverallEmissive}
   B=MaterialFunctionCall(LDR Input=Saturate(None=@n26), HDR Tint=Multiply(A=@n25) {constB=250}) {materialFunction=&DeriveHDRfromLDR}.RemappedRange
  )
  n27 = TextureSampleParameter2D() {parameterName=Diffuse, texture=&T_ClawMarks_D_Metal}
  Multiply(A=@n27.A, B=ScalarParameter() {defaultValue=1, parameterName=DecalOpacity})
  ScalarParameter() {parameterName=Metallic}
  ScalarParameter() {defaultValue=0.1, parameterName=Spec}
  ScalarParameter() {defaultValue=2, parameterName=Roughness}
  n28 = ObjectOrientation()
  n29 = Multiply(
   A=Multiply(A=@n16, B=ScalarParameter() {defaultValue=1, parameterName=OverallOpacity, sliderMax=1})
   B=OneMinus(None=MaterialFunctionCall!(TangentNormal, SlopAngle=@n28, FalloffPower, CheapContrast) {materialFunction=/Engine/Functions/Engine_MaterialFunctions01/Texturing/SlopeMask}.Result)
  )
  n30 = Multiply(A=@n23, B=@n29)
  n31 = ScalarParameter() {parameterName=FadeBaseColor}
  StaticSwitchParameter(
   True=TextureSample() {samplerType=SAMPLERTYPE_Normal, texture=&FlatNormalBlue}.RGB
   False=FeatureLevelSwitch(Default=@n3.Normal, ES2=@n4.A, ES3_1=@n4.A, SM5=LinearInterpolate(A=@n29, B=@n30, Alpha=@n31))
  ) {parameterName=DisableNormal}
  n32 = Constant()
  StaticSwitchParameter(True=@n1.RGB, False=FeatureLevelSwitch(Default=@n14.RGB, ES2=@n32, ES3_1=@n32, SM4_REMOVED=@n14.RGB)) {parameterName=UseSimpleColor}
  n33 = ScalarParameter() {defaultValue=1, parameterName=Detail_Intensity}
  n34 = TextureSampleParameter2D() {parameterName=Normal, samplerType=SAMPLERTYPE_Normal, texture=&T_ClawMarks_N}
  n35 = MaterialFunctionCall(
   Scale=ScalarParameter() {defaultValue=1, parameterName=Detail_Tiling}
   Diffuse=@n27.RGB
   DetailDiffuse=TextureObject() {texture=/Game/Global/Textures/Default/Blanks/T_FlatGrey}
   DiffuseIntensity=@n33
   Normal=@n34.RGB
   DetailNormal=TextureObjectParameter() {parameterName=Detail_Normal, samplerType=SAMPLERTYPE_Normal, texture=&T_Dudebro_Noise_N_01}
   NormalIntensity=@n33
  ) {materialFunction=&DetailTexturing}
  StaticSwitchParameter(True=@n35.Normal, False=@n34.RGB) {defaultValue=true, parameterName=UseDetailTextures}
  n36 = StaticSwitchParameter(True=@n35.Diffuse, False=@n27.RGB) {defaultValue=true, parameterName=UseDetailTextures}
  n37 = VectorParameter() {parameterName=Morning, defaultValue=<rgba 0.1771 0.0891 0.0577 1>}
  n38 = VectorParameter() {parameterName=Day, defaultValue=<rgba 0.6458 0.6458 0.6458 1>}
  n39 = VectorParameter() {parameterName=Evening, defaultValue=<rgba 0.2448 0.1364 0.1176 1>}
  n40 = VectorParameter() {parameterName=Night, defaultValue=<rgba 0.1864 0.3986 0.5833 1>}
  StaticSwitchParameter(
   True=Multiply(
    A=@n36
    B=MaterialFunctionCall(Color_Morning=@n37.RGB, Color_Day=@n38.RGB, Color_Evening=@n39.RGB, Color_Night=@n40.RGB) {materialFunction=&Color_Time_of_Day}.Result
   )
   False=@n36
  ) {defaultValue=true, parameterName=UseTODMTint}
  TextureCoordinate()
 ]
}
